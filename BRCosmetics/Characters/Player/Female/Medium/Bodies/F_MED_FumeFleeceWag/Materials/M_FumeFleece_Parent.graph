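# Parent character material for the FumeFleece skinned outfit, layering an optional animated "Good Mood" pattern onto the shared 2023 character material.
# PreFX output passes through thin-film, the switchable pattern, then optional crystal and gemstone layers into PostFX.
# Two panning noises and time sine-warp a third noise lookup that indexes a gradient.
# The tinted, Fresnel-attenuated result drives base color, its red channel roughness and specular, with mask-based blending and optional metallic override.
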Material /BRCosmetics/Characters/Player/Female/Medium/Bodies/F_MED_FumeFleeceWag/Materials/M_FumeFleece_Parent
graph {
 properties {
  bUseMaterialAttributes=true
  bUsedWithMeshDeformer=true
  bUsedWithMorphTargets=true
  bUsedWithSkeletalMesh=true
  pixelDepthOffsetMode=PDOM_Legacy
  shadingModel=MSM_FromMaterialExpression
  subsurfaceProfile=&SS_HeroSkin_02
  usedShadingModels=Default Lit | Subsurface | Clear Coat | Subsurface Profile | Cloth
 }
 nodes [
  n1 = ComponentMask(None=NamedRerouteUsage()) {b=true, r=true}
  n2 = ScalarParameter() {defaultValue=128, group=Good Mood, parameterName=GM_TextureScale}
  n3 = VectorParameter() {group=Good Mood, parameterName=GM_NoisePanningSpeed1, defaultValue=<rgba 0.018 0.025 0 1>}
  n4 = TextureSample(UVs=Panner(Coordinate=Divide(A=@n1, B=@n2) {constB=64}, Speed=ComponentMask(None=@n3.RGB) {g=true, r=true}) {speedX=0.018, speedY=0.025}) {
   samplerSource=SSM_Wrap_WorldGroupSettings
   samplerType=SAMPLERTYPE_LinearColor
   texture=/Game/Characters/Player/Female/Medium/Bodies/F_MED_GoodMood/Textures/T_GoodMood_Body_Noise
  }
  n5 = ComponentMask(None=NamedRerouteUsage()) {g=true, r=true}
  n6 = VectorParameter() {group=Good Mood, parameterName=GM_NoisePanningSpeed2, defaultValue=<rgba -0.01 0.03 0 1>}
  n7 = TextureSample(UVs=Panner(Coordinate=Divide(A=@n5, B=@n2) {constB=64}, Speed=ComponentMask(None=@n6.RGB) {g=true, r=true}) {speedX=-0.01, speedY=0.03}) {
   samplerSource=SSM_Wrap_WorldGroupSettings
   samplerType=SAMPLERTYPE_LinearColor
   texture=&T_GoodMood_Body_Noise
  }
  n8 = Normalize(VectorInput=VertexInterpolator(VS=PreSkinnedNormal()).PS)
  n9 = Power(Base=ComponentMask(None=Abs(None=@n8)) {r=true}) {constExponent=3}
  n10 = SphereMask(A=NamedRerouteUsage(), B=Constant3Vector() {constant=<rgba 40 0 23 0>}) {attenuationRadius=37, hardnessPercent=0}
  n11 = SphereMask(A=NamedRerouteUsage(), B=Constant3Vector() {constant=<rgba -40 0 23 0>}) {attenuationRadius=37, hardnessPercent=0}
  n12 = Multiply(A=Power(Base=Add(A=@n10, B=@n11)) {constExponent=3}) {constB=25}
  n13 = Add(
   A=Saturate(None=Add(A=@n9) {constB=0.5, desc=BIAS})
   B=StaticSwitchParameter(True=Saturate(None=@n12), False=Constant()) {group=Good Mood, parameterName=GM_UseSphereMaskForProjection}
  )
  LinearInterpolate(A=@n4.R, B=@n7.R, Alpha=SmoothStep(Value=Saturate(None=@n13)))
  n14 = MaterialFunctionCall() {materialFunction=&MF_CharacterParent_2023_PreFX}
  n15 = MaterialFunctionCall!(Material Attributes=@n14.Result, Fresnel=@n14.Result, Alpha=@n14.Result) {materialFunction=/Game/Characters/Player/Female/Medium/Bodies/F_MED_SkeleProbe/Materials/MF_ThinFilmDuplicate_TechArt}
  n16 = GetMaterialAttributes(None=@n15.Result) {attributeGetTypes [
    #1
    #2
    #3
    #4
   ]}
  n17 = SetMaterialAttributes!(
   MaterialAttributes=@n16.MaterialAttributes
   Base Color=NamedRerouteUsage()
   Roughness=NamedRerouteUsage()
   Specular=NamedRerouteUsage()
   Metallic=NamedRerouteUsage()
  ) {attributeSetTypes [
    #1
    #2
    #3
    #4
   ]}
  n18 = StaticSwitchParameter(True=@n17, False=@n15.Result) {group=Good Mood, parameterName=GM_UseAnimatedPattern}
  n19 = StaticSwitchParameter(
   True=MaterialFunctionCall!(Material=@n18, Alpha=NamedRerouteUsage()) {materialFunction=/Game/Global/MaterialFunctions/TechArt/MF_Crystal}.Result
   False=@n18
  ) {group=Good Mood, parameterName=GM_UseCrystal}
  n20 = NamedRerouteUsage()
  n21 = NamedRerouteUsage()
  MaterialFunctionCall(
   Material=StaticSwitchParameter(
    True=MaterialFunctionCall!(Material=@n19, Mask Texture=@n20, TechArtMask=@n21) {materialFunction=/Game/Characters/Player/Male/Medium/Bodies/M_MED_DiamondHeart_Jewel/Material/MF_Gemstone}.Result
    False=@n19
   ) {group=Good Mood, parameterName=GM_UseGemstone}
   M=NamedRerouteUsage()
   Pre Skinned Local Position=@n14.Pre skinned Local Position
   Distance Blend=@n14.Pre skinned Local Position
   Fresnel=@n14.Pre skinned Local Position
  ) {materialFunction=&MF_CharacterParent_2023_PostFX}
  n22 = TextureSampleParameter2D() {group=Good Mood, parameterName=GM_TAMask2, texture=&T_BlacksRGB}
  NamedRerouteDeclaration(None=ChannelMaskParameter(None=StaticSwitchParameter(True=@n22.RGB, False=@n14.Tech Art Mask) {group=Good Mood, parameterName=GM_UseCustomTAMask}) {group=Good Mood, parameterName=GM_TA_Channel}.R) {name=TA_Mask, nodeColor=<rgba 0.7344 1 0 1>}
  NamedRerouteDeclaration(None=@n16.Base Color) {name=Basecolor_IN, nodeColor=<rgba 1 0 0.6562 1>}
  n23 = ComponentMask(None=NamedRerouteUsage()) {b=true, r=true}
  n24 = Multiply(A=@n4.R, B=@n7.R)
  n25 = Multiply(A=Time(), B=ScalarParameter() {defaultValue=0.015, group=Good Mood, parameterName=GM_WarpSpeed})
  n26 = Sine(None=Multiply(A=Add(A=@n24, B=@n25), B=ScalarParameter() {defaultValue=0.4, group=Good Mood, parameterName=GM_Repetition}))
  n27 = ScalarParameter() {defaultValue=0.4, group=Good Mood, parameterName=GM_WarpIntensity}
  n28 = TextureSample!(
   UVs=Add(A=Divide(A=@n23) {constB=64}, B=Multiply(A=@n26, B=@n27))
   Tex
   "DDX(UVs)"=NamedRerouteUsage()
   "DDY(UVs)"=NamedRerouteUsage()
   Apply View MipBias
  ) {
   mipValueMode=TMVM_Derivative
   samplerSource=SSM_Wrap_WorldGroupSettings
   samplerType=SAMPLERTYPE_LinearColor
   texture=&T_GoodMood_Body_Noise
  }
  n29 = TextureSample!(
   UVs=MaterialFunctionCall(X=@n28.R, Y=Constant() {r=0.5}) {materialFunction=&MakeFloat2}.Result
   Tex
   Apply View MipBias
  ) {texture=/Game/Characters/Player/Female/Medium/Bodies/F_MED_GoodMood/Textures/GoodMood_Gradient}
  n30 = Add(A=Multiply(A=@n29.R, B=@n29.R)) {constB=0.4}
  n31 = SmoothStep(Value=NamedRerouteUsage()) {constMax=0.6, constMin=0.1}
  NamedRerouteDeclaration(None=LinearInterpolate(A=NamedRerouteUsage(), B=@n30, Alpha=@n31)) {name=Roughness_OUT, nodeColor=<rgba 1 0.1641 0 1>}
  n32 = Constant() {r=1}
  NamedRerouteDeclaration(None=LinearInterpolate(
    A=NamedRerouteUsage()
    B=MaterialFunctionCall(In=@n30, Contrast=@n32) {materialFunction=&CheapContrast}.Result
    Alpha=@n31
   ) {constB=0.5}) {name=Spec_OUT, nodeColor=<rgba 0.7109 0 1 1>}
  NamedRerouteDeclaration(None=@n14.M) {name=MaskTexture, nodeColor=<rgba 1 0 0.5391 1>}
  NamedRerouteDeclaration(None=@n16.Metallic) {name=Metallic_IN, nodeColor=<rgba 0.9922 1 0 1>}
  n33 = NamedRerouteUsage()
  n34 = ScalarParameter() {defaultValue=1, group=Good Mood, parameterName=GM_Metallic}
  NamedRerouteDeclaration(None=LinearInterpolate(A=@n33, B=StaticSwitchParameter(True=@n34, False=@n33) {group=Good Mood, parameterName=GM_UseCustomMetallic}, Alpha=@n31)) {name=Metallic_OUT, nodeColor=<rgba 0 1 0.9531 1>}
  n35 = ScalarParameter() {defaultValue=0.6, group=Good Mood, parameterName=GM_TintIntensity}
  n36 = ScalarParameter() {group=Good Mood, parameterName=GM_UseSphereMask}
  n37 = NamedRerouteUsage()
  n38 = Constant3Vector() {constant=<rgba 0 -5 85 0>}
  n39 = Multiply(A=Power(Base=SphereMask(A=@n37, B=@n38) {attenuationRadius=40, hardnessPercent=0}) {constExponent=3}) {constB=1.8}
  n40 = LinearInterpolate(
   A=@n29.RGB
   B=VectorParameter() {group=Good Mood, parameterName=GM_TintColor, defaultValue=<rgba 0.0271 0.0114 0.0677 1>}.RGB
   Alpha=LinearInterpolate(A=@n35, B=LinearInterpolate(A=@n35, Alpha=@n36), Alpha=Saturate(None=@n39))
  )
  n41 = OneMinus(None=Fresnel() {baseReflectFraction=0, exponent=2})
  NamedRerouteDeclaration(None=LinearInterpolate(A=NamedRerouteUsage(), B=Multiply(A=@n40, B=@n41), Alpha=@n31)) {name=BaseColor_OUT, nodeColor=<rgba 1 0.2109 0 1>}
  NamedRerouteDeclaration(None=@n14.Pre skinned Local Position) {name=PreskinLocalPos, nodeColor=<rgba 0 0.4141 1 1>}
  NamedRerouteDeclaration(None=ChannelMaskParameter(None=@n14.Tech Art Mask) {group=Good Mood, parameterName=GM_Crystal_Channel}.R) {name=CrystalChannel, nodeColor=<rgba 0.7344 1 0 1>}
  n42 = TextureCoordinate()
  NamedRerouteDeclaration(None=DDX(Value=@n42)) {name=ddx, nodeColor=<rgba 0.9453 0 1 1>}
  NamedRerouteDeclaration(None=DDY(Value=@n42)) {name=ddy, nodeColor=<rgba 0 0.7187 1 1>}
  n43 = NamedRerouteUsage()
  n44 = VectorParameter() {group=Good Mood, parameterName=CylindricalProj_Tiling, channelNames {a="", b="", g=V, r=U}, defaultValue=<rgba 2 100 0 1>}
  n45 = Multiply(
   A=MaterialFunctionCall(Vector or UVs=ComponentMask(None=@n43) {g=true, r=true}) {materialFunction=&VectorToRadialValue}.Vector Converted to Angle
   B=@n44.U
  ) {constB=2}
  n46 = ComponentMask(None=Divide(A=@n43, B=@n44.V)) {b=true}
  n47 = VectorParameter() {
   group=Good Mood
   parameterName=CylindricalProj_RotationSpeed
   channelNames {a="", b="", g=V_Speed, r=U_Speed}
   defaultValue=<rgba 0 0.02 0 1>
  }
  n48 = MaterialFunctionCall(X=@n47.U_Speed, Y=@n47.U_Speed) {materialFunction=&MakeFloat2}
  n49 = Time()
  NamedRerouteDeclaration(None=Add(A=AppendVector(A=@n45, B=@n46), B=Multiply(A=@n48.Result, B=@n49))) {name=UVs, nodeColor=<rgba 0 0.4375 1 1>}
  NamedRerouteDeclaration(None=@n16.Roughness) {name=Roughness_IN, nodeColor=<rgba 0 0.8125 1 1>}
  NamedRerouteDeclaration(None=@n16.Specular) {name=Spec_IN, nodeColor=<rgba 0.4531 0 1 1>}
 ]
}
